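# Physics asset for the DistantEchoPilot backpack cosmetic, setting up physical simulation for the ropes hanging from the pack.
# Small capsule bodies cover the pack, a root rope bone, and two branches of two and three segments.
# Constraints chain the root rope to the pack with tight swing, then fork into both branches with looser swing.
# The shorter branch barely twists, while the longer one allows equal swing and twist.
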PhysicsAsset /BRCosmetics/Accessories/FORT_Backpacks/Backpack_M_MED_DistantEchoPilot/Meshes/M_MED_DistantEchoPilot_Pack_Physics
bodies {
 pack [{shapeType=Capsule, length=1, radius=1, shapeName=pack_capsule}]
 rope_00 [{shapeType=Capsule, length=0.5, radius=1, shapeName=rope_00_capsule, center=<1 0 0>, rotation=<rot 90 0 0>}]
 rope_a_00 [{shapeType=Capsule, length=0.25, radius=0.5, shapeName=rope_a_00_capsule, center=<1 0 0>, rotation=<rot 90 0 0>}]
 rope_a_01 [{shapeType=Capsule, length=0.25, radius=0.5, shapeName=rope_a_01_capsule, center=<1.5 0 0>, rotation=<rot 0 -90 -90>}]
 rope_b_00 [{shapeType=Capsule, length=0.25, radius=0.5, shapeName=rope_b_00_capsule, center=<1 0 0>, rotation=<rot 90 0 0>}]
 rope_b_01 [{shapeType=Capsule, length=0.25, radius=0.5, shapeName=rope_b_01_capsule, center=<1 0 0>, rotation=<rot 90 0 0>}]
 rope_b_02 [{shapeType=Capsule, length=0.25, radius=0.5, shapeName=rope_b_02_capsule, center=<1 0 0>, rotation=<rot 90 0 0>}]
}
constraints [
 {bone1Name=rope_a_00, bone2Name=rope_00, swing1LimitDegrees=30, swing2LimitDegrees=30, twistLimitDegrees=1}
 {bone1Name=rope_00, bone2Name=pack, swing1LimitDegrees=10, swing2LimitDegrees=10, twistLimitDegrees=1}
 {bone1Name=rope_a_01, bone2Name=rope_a_00, swing1LimitDegrees=45, swing2LimitDegrees=45, twistLimitDegrees=1}
 {bone1Name=rope_b_00, bone2Name=rope_00, swing1LimitDegrees=30, swing2LimitDegrees=30, twistLimitDegrees=30}
 {bone1Name=rope_b_01, bone2Name=rope_b_00, swing1LimitDegrees=30, swing2LimitDegrees=30, twistLimitDegrees=30}
 {bone1Name=rope_b_02, bone2Name=rope_b_01, swing1LimitDegrees=30, swing2LimitDegrees=30, twistLimitDegrees=30}
]
